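# Niagara system for the swing trails of the MoleAcornGloam pickaxe, drawing short-lived ribbons that follow skeletal mesh bones through the shared AnimTrail modules.
# EnergyTrail uses two ribbon renderers, with User.EnableWeakPoint switching its offset along ribbon facing and its color between bright blue and neutral white-grey.
# MidEnergyTrail adds a narrower orange ribbon whose alpha also scales with System.ReactivityAlpha.
# Both trails are tinted by time of day and faded by System.ScaleAlpha.
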NiagaraSystem /BRCosmetics/Weapons/FORT_Melee/Pickaxe_MoleAcornGloam/FX/NS_Pickaxe_MoleAcornGloam_AnimTrails
dependencies {
 usedDataInterfaces [&NiagaraDataInterfaceColorCurve, &NiagaraDataInterfaceSkeletalMesh, &NiagaraDataInterfaceCurve]
 usedDynamicInputs [&Boolean_NotOperation, &FloatFromCurve]
 usedModules [
  &self:SystemSpawnScript.SystemScriptSource.SystemScriptGraph.NiagaraNodeAssignment_0.SetVariables_#1
  &self:SystemSpawnScript.SystemScriptSource.SystemScriptGraph.NiagaraNodeAssignment_5.SetVariables_#2
  &self:SystemSpawnScript.SystemScriptSource.SystemScriptGraph.NiagaraNodeAssignment_3.SetVariables_#3
  &SystemState
  &Timeline
  &self:SystemSpawnScript.SystemScriptSource.SystemScriptGraph.NiagaraNodeAssignment_2.SetVariables_#4
  &self:SystemSpawnScript.SystemScriptSource.SystemScriptGraph.NiagaraNodeAssignment_1.SetVariables_#5
 ]
}
emitters {
 EnergyTrail {
  emitterUpdateScript [
   &EmitterState()
   &FN_SpawnAnimTrail(AnimTrailWidthMode=::FromSecond)
  ]
  particleSpawnScript [
   &FN_InitializeAnimTrailBoneIndecies()
   &FN_InitializeAnimTrail(AnimTrailWidth=0.4, Lifetime=0.1, SkeletalMesh=$System.SkeletalMesh)
   /SetVariables(Particles.Position=&AddVectorToPosition(
     Position=$Particles.Position
     Vector=&Multiply_VectorByFloat(
      Vector=$Particles.RibbonFacing
      Float=&MakeCustomFloatFromBool(Bool=$User.EnableWeakPoint, True Float=-5.5, False Float=-10)
     )
    )) {isSetParameters=true, module=SetVariables_#6}
   /SetVariables() {isSetParameters=true}
   &AddVelocity(Velocity=$User.DebugSpeed)
  ]
  particleUpdateScript [
   &FN_UpdateAnimTrail(SkeletalMesh=$System.SkeletalMesh) {module=FN_UpdateAnimTrail001}
   &Color(Color=&MakeCustomLinearColorFromBool(
     Bool=$User.EnableWeakPoint
     True Color=&ColorFromCurve(ColorCurve=DataInterface() {propertyValues {
        AlphaCurve {keys [<at 0 1>,<at 1 0 RCIM_Cubic>]}
        BlueCurve {keys [<at 0 2>,<at 0.1509 3>]}
        GreenCurve {keys [<at 0 0.2542>,<at 0.1509 0.3734>]}
        RedCurve {keys [<at 0 0>,<at 0.1509 0.1826>]}
       }})
     False Color=&ColorFromCurve(ColorCurve=DataInterface() {propertyValues {
        AlphaCurve {keys [<at 0 1>,<at 0.1813 1>,<at 1 0 RCIM_Cubic>]}
        BlueCurve {keys [<at 0 0.8196>,<at 0.1509 0.5>]}
        GreenCurve {keys [<at 0 0.8486>,<at 0.1509 0.5>]}
        RedCurve {keys [<at 0 0.8646>,<at 0.1509 0.5>]}
       }})
    ))
   /SetVariables(Particles.Color=&MultiplyLinearColor(
     Linear Color=&MultiplyLinearColor(
      Linear Color=$Particles.Color
      Scale Factor=&MakeVector4_FromLinearColor(LinearColor=&MakeCustomLinearColorFromBool(
        Bool=$User.EnableWeakPoint
        True Color=&FNTimeOfDayColor(Night Color=<rgba 0.224 0.2818 1 1>, Morning Color=<rgba 1 1 1 1>, Day Color=<rgba 1 1 1 1>, Evening Color=<rgba 1 1 1 1>)
        False Color=&FNTimeOfDayColor(
         Night Color=<rgba 0.224 0.2818 1 1>
         Morning Color=<rgba 1 0.8246 0.5699 1>
         Day Color=<rgba 1 1 1 1>
         Evening Color=<rgba 1 0.8246 0.5699 1>
        )
       ))
     )
     Scale Factor=&MakeVector4(X=1, Y=1, Z=1, W=$System.ScaleAlpha)
    )) {isSetParameters=true, module=SetVariables_#7}
   &SolveForcesAndVelocity(Clamp Velocity=false, Limit Acceleration=false)
  ]
  rendererClasses [&NiagaraRibbonRendererProperties]
  renderers [
   {rendererClass=&NiagaraRibbonRendererProperties}
   {rendererClass=&NiagaraRibbonRendererProperties, rendererIndex=1}
  ]
 }
 MidEnergyTrail {
  emitterUpdateScript [
   &EmitterState()
   &FN_SpawnAnimTrail(AnimTrailWidthMode=::FromSecond)
  ]
  particleSpawnScript [
   &FN_InitializeAnimTrailBoneIndecies()
   &FN_InitializeAnimTrail(AnimTrailWidth=0.25, Lifetime=0.1, SkeletalMesh=$System.SkeletalMesh)
   /SetVariables(Particles.Position=&AddVectorToPosition(Position=$Particles.Position, Vector=&Multiply_VectorByFloat(Vector=$Particles.RibbonFacing, Float=-5.5))) {isSetParameters=true, module=SetVariables_#6}
   &AddVelocity(Velocity=$User.DebugSpeed)
  ]
  particleUpdateScript [
   &FN_UpdateAnimTrail(SkeletalMesh=$System.SkeletalMesh) {module=FN_UpdateAnimTrail001}
   &Color(Color=&ColorFromCurve(ColorCurve=DataInterface() {propertyValues {
       AlphaCurve {keys [<at 0 0.25>,<at 0.1841 0 RCIM_Cubic>]}
       BlueCurve {keys [<at 0 0.0007>,<at 0.1509 0.0159>]}
       GreenCurve {keys [<at 0 0.1815>,<at 0.1509 0.4806>]}
       RedCurve {keys [<at 0 2>,<at 0.1509 3>]}
      }}))
   /SetVariables(Particles.Color=&MultiplyLinearColor(
     Linear Color=&MultiplyLinearColor(
      Linear Color=$Particles.Color
      Scale Factor=&MakeVector4_FromLinearColor(LinearColor=&FNTimeOfDayColor(Night Color=<rgba 0.224 0.2818 1 1>, Morning Color=<rgba 1 1 1 1>, Day Color=<rgba 1 1 1 1>, Evening Color=<rgba 1 1 1 1>))
     )
     Scale Factor=&MakeVector4(X=1, Y=1, Z=1, W=&Multiply_Float(A=$System.ScaleAlpha, B=$System.ReactivityAlpha))
    )) {isSetParameters=true, module=SetVariables_#7}
   &SolveForcesAndVelocity(Clamp Velocity=false, Limit Acceleration=false)
  ]
  rendererClasses [&NiagaraRibbonRendererProperties]
  renderers [{rendererClass=&NiagaraRibbonRendererProperties}]
 }
}
summary {
 systemName=NS_Pickaxe_MoleAcornGloam_AnimTrails
 emitters [
  {emitterName=EnergyTrail, rendererClasses [&NiagaraRibbonRendererProperties]}
  {emitterName=MidEnergyTrail, rendererClasses [&NiagaraRibbonRendererProperties]}
 ]
 userVariables [
  User.DebugSpeed {type=Vector3f, description="", defaultValue {struct=Vector3f, value=<0 0 0>}}
  User.EnableWeakPoint {type=NiagaraBool, defaultValue=true}
  User.RibbonIDOffset {type=NiagaraInt32}
 ]
}
